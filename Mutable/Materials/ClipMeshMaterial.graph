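# ClipMeshMaterial is a minimal translucent, Default Lit material in the Mutable content, apparently meant for clip meshes going by its name.
# Its graph holds only a constant dark teal color, with no output connections, textures or parameters shown.
# It uses volumetric directional translucency lighting, casts no ray-traced shadows, and is flagged for skeletal mesh and mesh deformer use, with automatic usage flagging in the editor turned off.
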Material /Mutable/Materials/ClipMeshMaterial
graph {
 properties {
  bAutomaticallySetUsageInEditor=false
  bCastRayTracedShadows=false
  bUsedWithMeshDeformer=true
  bUsedWithSkeletalMesh=true
  blendMode=BLEND_Translucent
  pixelDepthOffsetMode=PDOM_Legacy
  translucencyLightingMode=TLM_VolumetricDirectional
  usedShadingModels=Default Lit
 }
 nodes [Constant3Vector() {constant=<rgba 0.0226 0.0812 0.0885 1>}]
}
